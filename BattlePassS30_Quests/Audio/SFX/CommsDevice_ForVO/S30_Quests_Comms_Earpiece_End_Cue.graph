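# Sound cue for the end sound of a comms earpiece device, made for Season 30 quest voice-over audio.
# A single wave player feeds the S30_Quests_Comms_Earpiece_End sound wave straight to the output, with no randomization or modulation.
# It plays through the narrator VO sound class and uses the shared audio-log attenuation asset from BattlePassBase with a very large max distance.
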SoundCue /BattlePassS30_Quests/Audio/SFX/CommsDevice_ForVO/S30_Quests_Comms_Earpiece_End_Cue
root=WavePlayer_0
nodes [{id=WavePlayer_0, properties {soundWaveAssetPtr=&/S30_Quests_Comms_Earpiece_End}}]
properties {
 attenuationSettings=/BattlePassBase/Audio/SFX/Fort_AudioLogs_Att
 maxDistance=2097152
 soundClassObject=/Game/Sounds/FortSoundClassesAndModes/Fort_VO_Narrator
 volumeMultiplier=1
}
